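# Material function for the StrideMiceDeep backpack that overrides emissive on incoming material attributes to render an animated, audio-reactive LED display.
# Five-frame flipbooks supply LED image and vignette color, with default and reactive variants for speed, brightness and glow.
# Bump-offset UVs and tiled bulb textures fake screen depth, with panned static, noise highlights and timed black fades.
# Named reroutes link the parts, including frontend-loop audio amplitude and pixel-depth distance fade.
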MaterialFunction /BRCosmetics/Accessories/FORT_Backpacks/Backpack_M_MED_StrideMiceDeep/Materials/MF_StrideMIceDeep_Led
graph {nodes [
  n1 = FunctionInput() {inputType=FunctionInput_MaterialAttributes}
  n2 = NamedRerouteUsage()
  n3 = NamedRerouteUsage()
  n4 = NamedRerouteUsage()
  n5 = NamedRerouteUsage()
  n6 = Add(
   A=Multiply(A=Multiply(A=@n2, B=@n3), B=ScalarParameter() {defaultValue=0.05, group=StrideMiceDeep_LED, parameterName=LEDBoiseBlend})
   B=Add(A=Multiply(A=@n4, B=@n5), B=NamedRerouteUsage())
  )
  n7 = Multiply(A=NamedRerouteUsage(), B=@n6)
  n8 = Multiply(A=@n6, B=Constant() {r=1}) {constB=0.7}
  n9 = NamedRerouteUsage()
  FunctionOutput(None=SetMaterialAttributes!(
    MaterialAttributes=GetMaterialAttributes(None=@n1) {attributeGetTypes [#1]}.MaterialAttributes
    EmissiveColor=LinearInterpolate(A=@n7, B=@n8, Alpha=@n9)
   ) {attributeSetTypes [#1]})
  n10 = Subtract(A=PixelDepth().R, B=Constant() {r=2048}) {constB=128}
  n11 = Constant() {r=20}
  NamedRerouteDeclaration(None=Saturate(None=Divide(A=@n10, B=@n11) {constB=1024})) {name=DistanceFade, nodeColor=<rgba 0 1 0.8125 1>}
  n12 = TextureCoordinate() {coordinateIndex=1}
  n13 = Constant2Vector() {g=0.1, r=5}
  n14 = NamedRerouteUsage()
  n15 = TextureSample(UVs=Panner(Coordinate=@n12, Speed=Multiply(A=@n13, B=@n14)) {speedX=5, speedY=0.1}) {
   samplerSource=SSM_Wrap_WorldGroupSettings
   samplerType=SAMPLERTYPE_Masks
   texture=/Game/Missions/Primary/RepairTheShelter/Textures/Cinematic_Screen_Static
  }
  n16 = Multiply(A=@n15.B) {constB=0.1}
  n17 = Constant2Vector() {g=-0.05, r=6}
  n18 = NamedRerouteUsage()
  n19 = TextureSample(UVs=Panner(Coordinate=@n12, Speed=Multiply(A=@n17, B=@n18)) {speedX=6, speedY=-0.05}) {
   samplerSource=SSM_Wrap_WorldGroupSettings
   samplerType=SAMPLERTYPE_Masks
   texture=&Cinematic_Screen_Static
  }
  n20 = Constant2Vector() {g=-0.03, r=-3}
  n21 = NamedRerouteUsage()
  n22 = TextureSample(UVs=Panner(Coordinate=@n12, Speed=Multiply(A=@n20, B=@n21)) {speedX=-3, speedY=-0.03}) {
   samplerSource=SSM_Wrap_WorldGroupSettings
   samplerType=SAMPLERTYPE_Masks
   texture=&Cinematic_Screen_Static
  }
  n23 = Multiply(A=@n19.G, B=@n22.R)
  NamedRerouteDeclaration(None=Multiply(A=Add(A=@n16, B=@n23), B=ScalarParameter() {defaultValue=6, group=MonitorStatic, parameterName=Static Scale})) {name=Static, nodeColor=<rgba 0 1 0.8828 1>}
  n24 = TextureSample() {constCoordinate=1, texture=/BRCosmetics/Accessories/FORT_Backpacks/Backpack_M_MED_StrideMiceDeep/Textures/T_LED_Mask}
  n25 = Multiply(
   A=Power(Base=NamedRerouteDeclaration(None=@n24.R) {name=LEDEdgeMask, nodeColor=<rgba 0.1484 1 0 1>}) {constExponent=1.5}
   B=LinearInterpolate(A=Constant() {r=0.35}, B=Constant() {r=1}, Alpha=NamedRerouteUsage()) {constB=0.58}
  )
  n26 = Multiply(A=TextureCoordinate() {coordinateIndex=1}, B=AppendVector(A=Constant() {r=1}, B=Constant() {r=1.2886}))
  n27 = AppendVector(A=Constant() {r=0.5}, B=Constant() {r=0.672})
  n28 = LinearInterpolate(
   A=ScalarParameter() {defaultValue=0.65, group=StrideMiceDeep_LED, parameterName=LED_Radius_Max}
   B=ScalarParameter() {defaultValue=0.62, group=StrideMiceDeep_LED, parameterName=LEDvignet_Min}
   Alpha=NamedRerouteUsage()
  ) {constB=0.58}
  n29 = Constant()
  n30 = Power(Base=OneMinus(None=SphereMask(A=@n26, B=@n27, Radius=@n28, Hardness=@n29)), Exp=Constant() {r=10}) {constExponent=4}
  n31 = ScalarParameter() {defaultValue=0.5, group=StrideMiceDeep_LED, parameterName=Edge glow}
  n32 = Multiply(A=Saturate(None=Max(A=@n25, B=@n30)), B=If(A=NamedRerouteUsage(), "A > B"=@n31, "A == B"=@n31, "A < B"=Constant() {r=0.8}) {constB=0.01})
  n33 = NamedRerouteUsage()
  NamedRerouteDeclaration(None=LinearInterpolate(A=@n32, B=Max(A=@n32, B=@n33), Alpha=NamedRerouteUsage())) {name=LEDvignetMask, nodeColor=<rgba 0 0.4609 1 1>}
  n34 = TextureCoordinate() {coordinateIndex=1}
  n35 = NamedRerouteDeclaration(None=ScalarParameter() {defaultValue=100, group=StrideMiceDeep_LED, parameterName=Light bulbTiling}) {name=Light bulbTiling, nodeColor=<rgba 1 0 0.1641 1>}
  n36 = ScalarParameter() {defaultValue=-1, group=StrideMiceDeep_LED, parameterName=LedBumpoffsetHeight}
  n37 = TextureSample(UVs=BumpOffset(Coordinate=Multiply(A=@n34, B=@n35), Height=NamedRerouteDeclaration(None=@n36) {name=LedBumpoffsetHeight, nodeColor=<rgba 0 1 0.7656 1>})) {samplerType=SAMPLERTYPE_Masks, texture=/BRCosmetics/Accessories/FORT_Backpacks/Backpack_M_MED_StrideMiceDeep/Textures/T_LEDLight}
  n38 = Multiply(A=@n37.R) {constB=0.8}
  n39 = TextureCoordinate() {coordinateIndex=1}
  n40 = TextureSample(UVs=BumpOffset(Coordinate=Multiply(A=@n35, B=@n39), Height=Constant())) {samplerType=SAMPLERTYPE_Masks, texture=&T_LEDLight}
  n41 = ScalarParameter() {defaultValue=3, group=StrideMiceDeep_LED, parameterName="Light bulb  Brightness"}
  NamedRerouteDeclaration(None=Add(A=Max(A=@n38, B=@n40.G), B=Multiply(A=@n37.B, B=@n41) {constB=2})) {name=Light bulb, nodeColor=<rgba 0.7812 1 0 1>}
  n42 = Time()
  n43 = MaterialFunctionCall(Multiplier=Constant() {r=0.08}) {materialFunction=&Pi}
  n44 = Sine(None=Subtract(A=Multiply(A=@n42, B=@n43.Result) {constB=5}) {constB=4.7124}) {period=6.2832}
  n45 = Multiply(A=Saturate(None=Subtract(A=@n44) {constB=0.992})) {constB=125}
  NamedRerouteDeclaration(None=OneMinus(None=Saturate(None=@n45))) {name=Blackfadein, nodeColor=<rgba 0 0.2266 1 1>}
  n46 = Time()
  n47 = MaterialFunctionCall(Multiplier=Constant() {r=10}) {materialFunction=&Pi}
  n48 = Sine(None=Add(A=Multiply(A=@n46, B=@n47.Result) {constB=5}) {constB=1.7}) {period=6.2832}
  NamedRerouteDeclaration(None=OneMinus(None=Saturate(None=@n48))) {name=BlackfadeinReactive, nodeColor=<rgba 1 0.0234 0 1>}
  n49 = Constant() {r=20}
  n50 = TextureSample(UVs=Panner(Coordinate=TextureCoordinate() {coordinateIndex=1, uTiling=0.5}) {speedY=0.1}) {
   samplerSource=SSM_Wrap_WorldGroupSettings
   samplerType=SAMPLERTYPE_Grayscale
   texture=/Game/Global/Textures/Default/Noise/LowResBlurredNoiseNormalized
  }
  n51 = Power(Base=@n50.R, Exp=Constant() {r=4})
  n52 = NamedRerouteUsage()
  n53 = Multiply(
   A=NamedRerouteDeclaration(None=Time()) {name=LEDTime, nodeColor=<rgba 0.7578 0 1 1>}
   B=ScalarParameter() {defaultValue=0.008, group=StrideMiceDeep_LED, parameterName=DefaultSpeed}
  ) {constB=0.1}
  n54 = Time()
  n55 = TextureObjectParameter() {
   group=StrideMiceDeep_LED
   parameterName=StrideMiceLED
   texture=/BRCosmetics/Accessories/FORT_Backpacks/Backpack_M_MED_StrideMiceDeep/Textures/T_StrideMiceDeep_LED
  }
  n56 = MaterialFunctionCall(
   "Animation  Phase (0-1)"=LinearInterpolate(A=NamedRerouteDeclaration(None=@n53) {name=LedAnimationPhase, nodeColor=<rgba 0 1 0.1797 1>}, B=Frac(None=@n54), Alpha=NamedRerouteUsage())
   Number of Rows=Constant() {r=1}
   Number of Columns=Constant() {r=5}
   Texture=@n55
   UVs=NamedRerouteUsage()
  ) {materialFunction=&FlipBook}
  n57 = NamedRerouteUsage()
  n58 = LinearInterpolate(
   A=Multiply(A=Multiply(A=@n52, B=@n56.Result)) {constB=1.7}
   B=Multiply(A=Multiply(A=@n57, B=@n56.Result), B=ScalarParameter() {defaultValue=3, group=StrideMiceDeep_LED, parameterName=HelmetReactiveBrightness})
   Alpha=NamedRerouteUsage()
  )
  n59 = MaterialFunctionCall!(Blend=Multiply(A=LinearInterpolate(B=@n49, Alpha=@n51) {constB=2}, B=Power(Base=@n58) {constExponent=1.5}), Base=@n58) {materialFunction=/Engine/Functions/Engine_MaterialFunctions03/Blends/Blend_Lighten}
  n60 = LinearInterpolate(
   A=ScalarParameter() {defaultValue=1, group=StrideMiceDeep_LED, parameterName=LED_Brightness_Min}
   B=ScalarParameter() {defaultValue=2, group=StrideMiceDeep_LED, parameterName=LED_Brightness_Max}
   Alpha=NamedRerouteUsage()
  )
  n61 = OneMinus(None=NamedRerouteUsage())
  NamedRerouteDeclaration(None=Multiply(A=Multiply(A=@n59.Result, B=@n60), B=Power(Base=@n61))) {name=LEDColor, nodeColor=<rgba 1 0 0.1406 1>}
  n62 = NamedRerouteUsage()
  n63 = Multiply(A=Time(), B=ScalarParameter() {defaultValue=1, group=StrideMiceDeep_LED, parameterName=LEDReactiveSpeed})
  n64 = Constant() {r=1}
  n65 = If(A=NamedRerouteUsage(), "A > B"=@n64, "A == B"=@n64, "A < B"=Constant()) {constB=0.01}
  n66 = TextureObjectParameter() {
   group=StrideMiceDeep_LED
   parameterName=StrideMiceLED_Color
   texture=/BRCosmetics/Accessories/FORT_Backpacks/Backpack_M_MED_StrideMiceDeep/Textures/T_StrideMiceGiant_LEDColor
  }
  n67 = MaterialFunctionCall(
   "Animation  Phase (0-1)"=LinearInterpolate(A=NamedRerouteUsage(), B=Frac(None=@n63), Alpha=NamedRerouteDeclaration(None=@n65) {name=ReactiveON, nodeColor=<rgba 0 1 0.2734 1>})
   Number of Rows=Constant() {r=1}
   Number of Columns=Constant() {r=5}
   Texture=@n66
   UVs=TextureCoordinate() {coordinateIndex=1}
  ) {materialFunction=&FlipBook}
  n68 = NamedRerouteUsage()
  NamedRerouteDeclaration(None=LinearInterpolate(A=Multiply(A=@n62, B=@n67.Result), B=Multiply(A=@n68, B=@n67.Result), Alpha=NamedRerouteUsage())) {name=LEDvignetColor, nodeColor=<rgba 0 0.4609 1 1>}
  n69 = Constant() {r=1.4451}
  n70 = Constant() {r=1.2}
  n71 = BumpOffset(Coordinate=Multiply(A=TextureCoordinate() {coordinateIndex=1}, B=AppendVector(A=@n69, B=@n70)), Height=NamedRerouteUsage())
  n72 = AppendVector(A=Constant() {r=0.72}, B=Constant() {r=0.6})
  n73 = Constant() {r=0.7}
  n74 = Constant() {r=-1}
  n75 = Subtract(A=@n73, B=Constant() {r=0.4})
  n76 = Constant()
  n77 = Power(
   Base=Subtract(A=SphereMask(A=@n71, B=@n72, Radius=@n73, Hardness=@n74), B=SphereMask(A=@n71, B=@n72, Radius=@n75, Hardness=@n76))
   Exp=Constant() {r=2}
  ) {constExponent=4}
  n78 = ScalarParameter() {defaultValue=0.5, group=StrideMiceDeep_LED, parameterName=ReactiveGlowBrightness}
  NamedRerouteDeclaration(None=ShadingPathSwitch(Default=Multiply(A=@n77, B=@n78) {constB=1.5}, Mobile=Constant())) {name=LEDvignetMask_Reactive, nodeColor=<rgba 0 0.4609 1 1>}
  n79 = Constant() {r=0.8}
  n80 = Constant() {r=1}
  n81 = MaterialFunctionCall(
   UVs=TextureCoordinate() {coordinateIndex=1}
   Texture Scale=Multiply(A=AppendVector(A=@n79, B=@n80), B=ScalarParameter() {defaultValue=0.96, group=StrideMiceDeep_LED, parameterName=LEDUVScale}) {constB=0.96}
  ) {materialFunction=&ScaleUVsByCenter}
  n82 = AppendVector(A=ScalarParameter() {defaultValue=0.02, group=StrideMiceDeep_LED, parameterName=LEDUOffset}, B=Constant())
  n83 = NamedRerouteUsage()
  n84 = SphereMask(A=TextureCoordinate() {coordinateIndex=1}, B=Constant() {r=0.5}, Radius=Constant() {r=0.4})
  NamedRerouteDeclaration(None=BumpOffset(Coordinate=Add(A=@n81.UVs, B=@n82), Height=LinearInterpolate(B=@n83, Alpha=@n84))) {name=LEDUV, nodeColor=<rgba 1 0.1875 0 1>}
  NamedRerouteDeclaration(None=Multiply(
    A=MaterialFunctionCall!(InAmplitude, InSpectrum, CustomFakeSpectrum, CustomFakeAmplitude) {materialFunction=/BRCosmetics/Global/MaterialFunctions/AudioReactivity/MF_AudioReactivity_FrontendLoop}.Amplitude
    B=ScalarParameter() {defaultValue=1, group=StrideMiceDeep_LED, parameterName=EnableAudioReactivity, sliderMax=1}
   )) {name=Amplitude, nodeColor=<rgba 0 1 0.1094 1>}
  NamedRerouteDeclaration(None=ScalarParameter() {defaultValue=1, group=MonitorStatic, parameterName=StaticSpeed}) {name=StaticSpeed, nodeColor=<rgba 0 1 0.6484 1>}
 ]}
