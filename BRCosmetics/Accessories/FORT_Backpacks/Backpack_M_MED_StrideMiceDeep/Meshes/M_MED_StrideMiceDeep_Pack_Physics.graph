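# Physics asset for the StrideMiceDeep backpack cosmetic, providing collision shapes and joint limits for its left and right lower pipes.
# Each side chains a small base capsule to two dynamic pipe capsules through constraints with limited swing and zero-degree twist limits.
# Swing limits tighten from the first joint to the second and differ slightly between sides.
# A large offset capsule covers the pack bone and has no constraint.
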PhysicsAsset /BRCosmetics/Accessories/FORT_Backpacks/Backpack_M_MED_StrideMiceDeep/Meshes/M_MED_StrideMiceDeep_Pack_Physics
bodies {
 Base_LowerPipe_L [{shapeType=Capsule, length=0.505, radius=0.505, shapeName=Base_LowerPipe_L_capsule, rotation=<rot 90 0 0>}]
 Base_LowerPipe_R [{shapeType=Capsule, length=0.505, radius=0.505, shapeName=Base_LowerPipe_R_capsule, rotation=<rot 90 0 0>}]
 Dyn_LowerPipe_L_01 [{shapeType=Capsule, length=0.6859, radius=1.1796, shapeName=Dyn_LowerPipe_L_01_capsule, center=<0.2 0 0>, rotation=<rot 90 0 0>}]
 Dyn_LowerPipe_L_02 [{shapeType=Capsule, length=1.5788, radius=1.1796, shapeName=Dyn_LowerPipe_L_02_capsule, center=<0.25 0 0>, rotation=<rot 90 0 0>}]
 Dyn_LowerPipe_R_01 [{shapeType=Capsule, length=0.4156, radius=1.1796, shapeName=Dyn_LowerPipe_R_01_capsule, center=<0.3019 0 0>, rotation=<rot 90 0 0>}]
 Dyn_LowerPipe_R_02 [{shapeType=Capsule, length=3.41, radius=1.1796, shapeName=Dyn_LowerPipe_R_02_capsule, center=<0.5 0 0>, rotation=<rot 90 0 0>}]
 pack [{shapeType=Capsule, length=16.4156, radius=3.7859, center=<0.6799 -5.7538 -6.906>, rotation=<rot 90 0 0>}]
}
constraints [
 {bone1Name=Dyn_LowerPipe_R_01, bone2Name=Base_LowerPipe_R, swing1LimitDegrees=38, swing2LimitDegrees=45, twistLimitDegrees=0}
 {bone1Name=Dyn_LowerPipe_R_02, bone2Name=Dyn_LowerPipe_R_01, swing1LimitDegrees=30, swing2LimitDegrees=40, twistLimitDegrees=0}
 {bone1Name=Dyn_LowerPipe_L_02, bone2Name=Dyn_LowerPipe_L_01, swing1LimitDegrees=25, swing2LimitDegrees=40, twistLimitDegrees=0}
 {bone1Name=Dyn_LowerPipe_L_01, bone2Name=Base_LowerPipe_L, swing1LimitDegrees=30, swing2LimitDegrees=45, twistLimitDegrees=0}
]
