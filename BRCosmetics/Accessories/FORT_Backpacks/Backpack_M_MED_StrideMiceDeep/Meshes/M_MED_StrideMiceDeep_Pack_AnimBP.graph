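# Animation Blueprint for the StrideMiceDeep backpack cosmetic, adding physics-driven motion to its pipes and pack.
# The cached input pose gets additive pack offsets, partly driven by spine bend to correct placement.
# A front-end flag selects between an all-AnimDynamics pipe chain and upper-pipe AnimDynamics plus a RigidBody simulation.
# While skydiving, the EventGraph wobbles upper-pipe gravity using sine waves and Perlin noise, tuned for float or dive modes.
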AnimBlueprint /BRCosmetics/Accessories/FORT_Backpacks/Backpack_M_MED_StrideMiceDeep/Meshes/M_MED_StrideMiceDeep_Pack_AnimBP
target_skeleton=&/M_MED_StrideMiceDeep_Pack_Skeleton
graphs [
 AnimGraph {nodes [
   n1 = "Transform (Modify) Bone - Bone: Dyn_UpperPipe_L"!(ComponentPose, bAlphaBoolEnabled:bool=True, Translation="0,0,0", Rotation=Get RotatePipe_L() -> (RotatePipe_L:Rotator), Scale="1,1,1") -> (Pose) {
    alphaInputType=Bool
    rotationMode=BMM_Additive
    rotationSpace=BCS_BoneSpace
    alphaBoolBlend {blendInTime=0.3, blendOption=Linear, blendOutTime=0.8}
    boneToModify {boneName=Dyn_UpperPipe_L}
   }
   n2 = "Transform (Modify) Bone - Bone: Dyn_UpperPipe_R"!(ComponentPose=@n1, bAlphaBoolEnabled:bool=True, Translation="0,0,0", Rotation=Get RotatePipe_R() -> (RotatePipe_R:Rotator), Scale="1,1,1") -> (Pose) {
    alphaInputType=Bool
    rotationMode=BMM_Additive
    rotationSpace=BCS_BoneSpace
    alphaBoolBlend {blendInTime=0.3, blendOption=Linear, blendOutTime=0.8}
    boneToModify {boneName=Dyn_UpperPipe_R}
   }
   "Transform (Modify) Bone - Bone: pack"!(ComponentPose=@n2, bAlphaBoolEnabled:bool, Translation="0,0,0", Rotation=Get BP_rotator() -> (BP_rotator:Rotator), Scale) -> (Pose) {
    alphaInputType=Bool
    rotationMode=BMM_Additive
    rotationSpace=BCS_BoneSpace
    translationMode=BMM_Additive
    translationSpace=BCS_BoneSpace
    alphaBoolBlend {blendInTime=0.3, blendOption=Linear, blendOutTime=0.9}
    boneToModify {boneName=pack}
   }
   n3 = Anim Dynamics - Bone: Dyn_UpperPipe_R!(
    ComponentPose=Local To Component(LocalPose=Use cached pose 'CRCache_InputPose'())
    Alpha=1
    GravityOverride:Vector=Get GravityOverride()
    ExternalForce="0,0,0"
   ) -> (Pose) {
    angularDampingOverride=1
    angularSpringConstant=13000
    bAngularSpring=true
    bOverrideAngularDamping=true
    bUseGravityOverride=true
    gravityScale=0.35
    alphaBoolBlend {blendOption=Linear}
    boundBone {boneName=Dyn_UpperPipe_R}
    componentAppliedLinearAccClamp=<5000 5000 5000>
    componentLinearAccScale=<0.8 0.8 0.8>
    componentLinearVelScale=<0.8 0.8 0.8>
    gravityOverride {z=-980}
    physicsBodyDefinitions [{
      boundBone {boneName=Dyn_UpperPipe_R}
      boxExtents=<30 35 35>
      constraintSetup {angularLimitsMax=<0 23 32>, angularLimitsMin=<0 -40 -50>, angularTarget=<1 0 0>}
      localJointOffset=<-3.5 0 0>
     }]
   }
   n4 = Anim Dynamics - Bone: Dyn_UpperPipe_L!(ComponentPose=@n3, Alpha=1, GravityOverride:Vector=Get GravityOverride(), ExternalForce="0,0,0") -> (Pose) {
    angularDampingOverride=0.9999
    angularSpringConstant=20000
    bAngularSpring=true
    bOverrideAngularDamping=true
    bUseGravityOverride=true
    gravityScale=0.35
    alphaBoolBlend {blendOption=Linear}
    boundBone {boneName=Dyn_UpperPipe_L}
    componentAppliedLinearAccClamp=<5000 5000 5000>
    componentLinearAccScale=<0.8 0.8 0.8>
    componentLinearVelScale=<0.8 0.8 0.8>
    gravityOverride {z=-980}
    physicsBodyDefinitions [{
      boundBone {boneName=Dyn_UpperPipe_L}
      boxExtents=<30 35 35>
      constraintSetup {angularLimitsMax=<0 25 50>, angularLimitsMin=<0 -25 -35>, angularTarget=<1 0 0>}
      localJointOffset=<3.5 0 0>
     }]
    relativeSpaceBone {boneName=Base_UpperPipe_L}
    simSpaceSettings {simSpaceAngularAlpha=1}
   }
   n5 = RigidBody!(ComponentPose=@n4, Alpha, OverrideWorldGravity:Vector="0,0,-980", ExternalForce) -> (Pose) {
    bOverrideWorldGravity=true
    overridePhysicsAsset=&/M_MED_StrideMiceDeep_Pack_Physics
    simulationSpace=BaseBoneSpace
    alphaBoolBlend {blendOption=Linear}
    baseBoneRef {boneName=pack}
    componentAppliedLinearAccClamp=<3000 3000 3000>
    componentLinearAccScale=<0.3 0.3 0.3>
    componentLinearVelScale=<0.3 0.3 0.3>
    overrideWorldGravity {z=-980}
    simSpaceSettings {
     dampingAlpha=0
     maxAngularAcceleration=1200
     maxAngularVelocity=300
     maxLinearAcceleration=1000
     maxLinearVelocity=600
     velocityScaleZ=0.2
     worldAlpha=0.5
     externalLinearDragV=<0.2 0.2 0.2>
    }
   }
   n6 = Anim Dynamics - Bone: Dyn_UpperPipe_R(
    ComponentPose=Local To Component(LocalPose=Use cached pose 'CRCache_InputPose'())
    Alpha=1
    GravityOverride=Get GravityOverride()
    ExternalForce="0,0,0"
   ) {
    angularDampingOverride=1
    angularSpringConstant=13000
    bAngularSpring=true
    bOverrideAngularDamping=true
    bUseGravityOverride=true
    gravityScale=0.35
    alphaBoolBlend {blendOption=Linear}
    boundBone {boneName=Dyn_UpperPipe_R}
    componentAppliedLinearAccClamp=<5000 5000 5000>
    componentLinearAccScale=<0.8 0.8 0.8>
    componentLinearVelScale=<0.8 0.8 0.8>
    gravityOverride {z=-980}
    physicsBodyDefinitions [{
      boundBone {boneName=Dyn_UpperPipe_R}
      boxExtents=<30 35 35>
      constraintSetup {angularLimitsMax=<0 23 32>, angularLimitsMin=<0 -40 -50>, angularTarget=<1 0 0>}
      localJointOffset=<-3.5 0 0>
     }]
   }
   n7 = Anim Dynamics - Bone: Dyn_UpperPipe_L(ComponentPose=@n6, Alpha=1, GravityOverride=Get GravityOverride(), ExternalForce="0,0,0") {
    angularDampingOverride=0.9999
    angularSpringConstant=20000
    bAngularSpring=true
    bOverrideAngularDamping=true
    bUseGravityOverride=true
    gravityScale=0.35
    alphaBoolBlend {blendOption=Linear}
    boundBone {boneName=Dyn_UpperPipe_L}
    componentAppliedLinearAccClamp=<5000 5000 5000>
    componentLinearAccScale=<0.8 0.8 0.8>
    componentLinearVelScale=<0.8 0.8 0.8>
    gravityOverride {z=-980}
    physicsBodyDefinitions [{
      boundBone {boneName=Dyn_UpperPipe_L}
      boxExtents=<30 35 35>
      constraintSetup {angularLimitsMax=<0 25 50>, angularLimitsMin=<0 -25 -35>, angularTarget=<1 0 0>}
      localJointOffset=<3.5 0 0>
     }]
    relativeSpaceBone {boneName=Base_UpperPipe_L}
    simSpaceSettings {simSpaceAngularAlpha=1}
   }
   n8 = Anim Dynamics - Bone: Dyn_LowerPipe_L_01(ComponentPose=@n7, Alpha=1, ExternalForce="0,0,0") {
    angularDampingOverride=0.999
    angularSpringConstant=5000
    bAngularSpring=true
    bOverrideAngularDamping=true
    bUseGravityOverride=true
    gravityScale=0.35
    alphaBoolBlend {blendOption=Linear}
    boundBone {boneName=Dyn_LowerPipe_L_01}
    componentAppliedLinearAccClamp=<9000 9000 9000>
    componentLinearAccScale=<1 1 1>
    componentLinearVelScale=<1 1 1>
    gravityOverride {z=-980}
    physicsBodyDefinitions [{
      boundBone {boneName=Dyn_LowerPipe_L_01}
      boxExtents=<20 30 30>
      constraintSetup {angularLimitsMax=<0 30 30>, angularLimitsMin=<0 -30 -30>, angularTarget=<1 0 0>}
      localJointOffset=<1.5 0 0>
     }]
    simSpaceSettings {maxAngularAcceleration=100, maxAngularVelocity=3, simSpaceAngularAlpha=0.5}
   }
   n9 = Anim Dynamics - Bone: Dyn_LowerPipe_L_02(ComponentPose=@n8, Alpha=1, ExternalForce="0,0,0") {
    angularDampingOverride=0.999
    angularSpringConstant=5000
    bAngularSpring=true
    bOverrideAngularDamping=true
    bUseGravityOverride=true
    gravityScale=0.35
    alphaBoolBlend {blendOption=Linear}
    boundBone {boneName=Dyn_LowerPipe_L_02}
    componentAppliedLinearAccClamp=<9000 9000 9000>
    componentLinearAccScale=<1 1 1>
    componentLinearVelScale=<1 1 1>
    gravityOverride {z=-980}
    physicsBodyDefinitions [{
      boundBone {boneName=Dyn_LowerPipe_L_02}
      boxExtents=<20 30 30>
      constraintSetup {angularLimitsMax=<0 30 30>, angularLimitsMin=<0 -30 -30>, angularTarget=<1 0 0>}
      localJointOffset=<1.5 0 0>
     }]
    simSpaceSettings {maxAngularAcceleration=100, maxAngularVelocity=3, simSpaceAngularAlpha=0.5}
   }
   n10 = Anim Dynamics - Bone: Dyn_LowerPipe_R_01(ComponentPose=@n9, Alpha=1, ExternalForce="0,0,0") {
    angularDampingOverride=0.999
    angularSpringConstant=5000
    bAngularSpring=true
    bOverrideAngularDamping=true
    bUseGravityOverride=true
    gravityScale=0.35
    alphaBoolBlend {blendOption=Linear}
    boundBone {boneName=Dyn_LowerPipe_R_01}
    componentAppliedLinearAccClamp=<9000 9000 9000>
    componentLinearAccScale=<1 1 1>
    componentLinearVelScale=<1 1 1>
    gravityOverride {z=-980}
    physicsBodyDefinitions [{
      boundBone {boneName=Dyn_LowerPipe_R_01}
      boxExtents=<20 30 30>
      constraintSetup {angularLimitsMax=<0 30 30>, angularLimitsMin=<0 -30 -30>, angularTarget=<1 0 0>}
      localJointOffset=<1.5 0 0>
     }]
    simSpaceSettings {maxAngularAcceleration=100, maxAngularVelocity=3, simSpaceAngularAlpha=0.5}
   }
   n11 = Anim Dynamics - Bone: Dyn_LowerPipe_R_02(ComponentPose=@n10, Alpha=1, ExternalForce="0,0,0") {
    angularDampingOverride=0.999
    angularSpringConstant=5000
    bAngularSpring=true
    bOverrideAngularDamping=true
    bUseGravityOverride=true
    gravityScale=0.35
    alphaBoolBlend {blendOption=Linear}
    boundBone {boneName=Dyn_LowerPipe_R_02}
    componentAppliedLinearAccClamp=<9000 9000 9000>
    componentLinearAccScale=<1 1 1>
    componentLinearVelScale=<1 1 1>
    gravityOverride {z=-980}
    physicsBodyDefinitions [{
      boundBone {boneName=Dyn_LowerPipe_R_02}
      boxExtents=<20 30 30>
      constraintSetup {angularLimitsMax=<0 30 30>, angularLimitsMin=<0 -30 -30>, angularTarget=<1 0 0>}
      localJointOffset=<1.5 0 0>
     }]
    simSpaceSettings {maxAngularAcceleration=100, maxAngularVelocity=3, simSpaceAngularAlpha=0.5}
   }
   Output Pose(Result=Blend Poses by bool(BlendPose_0=Component To Local(ComponentPose=@n11), BlendPose_1=Component To Local(ComponentPose=@n5), bActiveValue=Get bIsInFrontEnd()) {blendType=Linear}) {layerGroup=None}
   n12 = "Transform (Modify) Bone - Bone: pack"!(ComponentPose=Local To Component(LocalPose=Input Pose()), bAlphaBoolEnabled, Translation="0,0.25,0", Rotation="0,0,7.5") -> (Pose) {
    alphaInputType=Bool
    rotationMode=BMM_Additive
    rotationSpace=BCS_BoneSpace
    translationMode=BMM_Additive
    translationSpace=BCS_BoneSpace
    alphaBoolBlend {blendInTime=0.5, blendOption=Linear, blendOutTime=0.5}
    boneToModify {boneName=pack}
    rotation {roll=7.5}
    translation {y=0.25}
   }
   n13 = "Transform (Modify) Bone - Bone: pack"!(ComponentPose=@n12, Alpha:float, Translation=Get backpack_translation(), Rotation=Get backpack_rotate(), Scale:Vector) -> (Pose) {rotationMode=BMM_Additive, translationMode=BMM_Additive, alphaBoolBlend {blendOption=Linear}, boneToModify {boneName=pack}}
   Save cached pose 'CRCache_InputPose'(Pose=Component To Local(ComponentPose=@n13)) {cacheName=CRCache_InputPose}
  ]}
 EventGraph {nodes [
   n1 = Event BlueprintUpdateAnimation()
   > n2 = Sequence!(execute) -> (then_0, then_1, then_2:Exec, then_3:Exec)
   n3 = Branch(execute=@n2.then_1, Condition=Get bIsSkydiveFloating())
   > Set Set Gravity(Set Gravity:Vector="0,0,600") -> (Output_Get:Vector)
   > Set SetAmplitud(SetAmplitud:Vector="150,200,500") -> (Output_Get:Vector)
   > Set SetFrequency(SetFrequency:Rotator="3,4,2") -> (Output_Get:Rotator)
   Set Set Gravity(execute=@n3.else, Set Gravity="0,0,-980")
   > Branch(Condition=Get bIsSkydiveDiveMode())
   > Set Set Gravity(Set Gravity="0,0,200")
   > Set SetAmplitud(SetAmplitud="1300,1300,600")
   > Set SetFrequency(SetFrequency="13,7,13")
   n4 = Branch(execute=@n2.then_2, Condition=Get bIsSkyDiving())
   n5 = BreakVector(InVec=Get SetAmplitud() -> (SetAmplitud:Vector))
   n6 = BreakRotator(InRot=Get SetFrequency() -> (SetFrequency:Rotator))
   n7 = GetGameTimeInSeconds()
   n8 = "Sin (Radians)"(A=float + float(A=float * float(A=@n6.Yaw, B=@n7) {is=title K2Node_PromotableOperator}, B=@n1.DeltaTimeX))
   n9 = float * float(A=PerlinNoise1D!(Value:float=@n7) -> (ReturnValue:float), B=1) {is=title K2Node_CommutativeAssociativeBinaryOperator}
   n10 = float * float(A=@n8, B=@n9)
   n11 = float + float(A=float * float(A=float + float(A=@n8, B=@n10) {is=title K2Node_CommutativeAssociativeBinaryOperator}, B=@n5.Z))
   n12 = "Sin (Radians)"(A=float + float(A=float * float(A=@n6, B=@n7) {is=title K2Node_PromotableOperator}, B=@n1.DeltaTimeX) {is=title K2Node_PromotableOperator})
   n13 = float * float(A=@n12, B=@n9) {is=title K2Node_CommutativeAssociativeBinaryOperator}
   n14 = float + float(A=float * float(A=float + float(A=@n12, B=@n13) {is=title K2Node_CommutativeAssociativeBinaryOperator}, B=@n5))
   n15 = "Sin (Radians)"(A=float + float(A=float * float(A=@n6.Pitch, B=@n7) {is=title K2Node_PromotableOperator}, B=@n1.DeltaTimeX) {is=title K2Node_PromotableOperator})
   n16 = float * float(A=@n15, B=@n9) {is=title K2Node_CommutativeAssociativeBinaryOperator}
   n17 = float + float(A=float * float(A=float + float(A=@n15, B=@n16) {is=title K2Node_CommutativeAssociativeBinaryOperator}, B=@n5.Y))
   > Set GravityOverride(GravityOverride=vector + vector(A=MakeVector(X=@n14, Y=@n17, Z=@n11), B=Get Set Gravity() -> (Set Gravity:Vector)))
   Set GravityOverride(execute=@n4.else, GravityOverride="0,0,-980")
   n18 = Get bIsSkyDiving()
   n19 = Get bIsPlayingEmote()
   Set Backpack Fix Needed(execute=@n2.then_3, Backpack Fix Needed:bool=AND Boolean(A=NOT Boolean(A=@n18), B=NOT Boolean(A=@n19))) -> (Output_Get:bool)
   n20 = GetOwnerSkeletalMeshComponent()
   n21 = BreakRotator(InRot=BreakTransform(InTransform=GetDeltaTransformFromRefPose(self=@n20, BoneName=spine_05, BaseName=spine_01)).Rotation)
   Set backpack_rotate(execute=@n2.then_0, backpack_rotate=MakeRotator(Roll=MapRangeClamped(Value=@n21.Yaw, InRangeA=0, InRangeB=-30, OutRangeB=35)))
   > Set backpack_translation(backpack_translation=MakeVector(Y=MapRangeClamped(Value=@n21.Yaw, InRangeA=0, InRangeB=-30, OutRangeB=-0.5)))
   TryGetPawnOwner()
  ]}
]
variables [
 Backpack Fix Needed {type=boolean}
 backpack_rotate {type=Rotator}
 backpack_translation {type=Vector}
 BP_rotator {type=Rotator}
 GravityOverride {type=Vector}
 RotatePipe_L {type=Rotator}
 RotatePipe_R {type=Rotator}
 Set Gravity {type=Vector}
 SetAmplitud {type=Vector}
 SetFrequency {type=Rotator}
]
